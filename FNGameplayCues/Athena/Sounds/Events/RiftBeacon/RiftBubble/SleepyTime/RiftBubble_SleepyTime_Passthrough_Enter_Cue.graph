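# Sound cue for entering a SleepyTime Rift Beacon bubble in Athena, judging by its name and path.
# A random node picks between two equally weighted waves, which are Western bubble enter sounds rather than SleepyTime-specific ones.
# It plays as Athena ambient audio with the safe zone shield attenuation.
# A concurrency override allows one instance per owner, stopping the oldest.
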SoundCue /FNGameplayCues/Athena/Sounds/Events/RiftBeacon/RiftBubble/SleepyTime/RiftBubble_SleepyTime_Passthrough_Enter_Cue
root=Random_0
nodes [
 {id=Random_0, children [WavePlayer_1,WavePlayer_2], properties {weights [1,1]}}
 {id=WavePlayer_1, properties {soundWaveAssetPtr=/FNGameplayCues/Athena/Sounds/Events/RiftBeacon/RiftBubble/Western/Western_Enter_01}}
 {id=WavePlayer_2, properties {soundWaveAssetPtr=/FNGameplayCues/Athena/Sounds/Events/RiftBeacon/RiftBubble/Western/Western_Enter_02}}
]
properties {
 attenuationSettings=/Game/Sounds/Attenuation/Fort_SafeZone_Shield_Athena_Att
 bOverrideConcurrency=true
 maxDistance=4000
 soundClassObject=&Fort_Ambient_Athena
 volumeMultiplier=1.247
 concurrencyOverrides {bLimitToOwner=true, maxCount=1, resolutionRule=StopOldest}
}
