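# Sound cue for music played on entering (passing through) the SleepyTime Rift Beacon rift bubble, using safe-zone shield attenuation.
# Randomly plays one of two autumn orchestral stingers under the Gadgets sound class.
# A quieter mixer branch under the music-pack class holds a fade-in enveloper with no sound connected, so it contributes nothing.
# Concurrency allows one instance per owner, stopping the oldest.
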SoundCue /FNGameplayCues/Athena/Sounds/Events/RiftBeacon/RiftBubble/SleepyTime/RiftBubble_SleepyTime_Passthrough_Enter_Music_Cue
root=Mixer_0
nodes [
 {
  id=Enveloper_0
  children [null]
  properties {pitchCurve {editorCurveData {keys [{time=0}]}}, volumeCurve {editorCurveData {keys [<at 0 0>,<at 7 1>]}}}
 }
 {id=Mixer_0, children [SoundClass_0,SoundClass_1], properties {inputVolume [1,0.3]}}
 {id=Random_0, children [WavePlayer_1,WavePlayer_2], properties {weights [1,1]}}
 {id=SoundClass_0, children [Random_0], properties {soundClassOverride=&Fort_Gadgets}}
 {id=SoundClass_1, children [Enveloper_0], properties {soundClassOverride=/Game/Sounds/FortSoundClassesAndModes/Fort_Music_MusicPacks}}
 {id=WavePlayer_1, properties {soundWaveAssetPtr=/FNGameplayCues/Sounds/Fort_Music/Orchestral_Score/Autumn/Music_Autumn_Amb_Day_A_Stinger3}}
 {id=WavePlayer_2, properties {soundWaveAssetPtr=/FNGameplayCues/Sounds/Fort_Music/Orchestral_Score/Autumn/Music_Autumn_Amb_Day_A_Stinger2}}
]
properties {
 attenuationSettings=/Game/Sounds/Attenuation/Fort_SafeZone_Shield_Athena_Att
 bOverrideConcurrency=true
 maxDistance=4000
 volumeMultiplier=1
 concurrencyOverrides {bLimitToOwner=true, maxCount=1, resolutionRule=StopOldest}
}
