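# Sound cue for the SleepyTime projectile opening in the Rift Beacon bubble event, judging by its name and path.
# A random node picks one of two equally weighted impact waves, and a modulator plays it at a fixed raised pitch.
# It uses the shared SleepyTime attenuation asset and the Athena ambient sound class.
# A concurrency override caps simultaneous instances, stopping the oldest first.
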SoundCue /FNGameplayCues/Athena/Sounds/Events/RiftBeacon/RiftBubble/SleepyTime/SleepyTime_Prj_Open_Cue
root=Modulator_0
nodes [
 {id=Modulator_0, children [Random_0], properties {pitchMax=1.2, pitchMin=1.2}}
 {id=Random_0, children [WavePlayer_0,WavePlayer_2], properties {weights [1,1]}}
 {id=WavePlayer_0, properties {soundWaveAssetPtr=&/SleepyTime_Impact_02}}
 {id=WavePlayer_2, properties {soundWaveAssetPtr=&/SleepyTime_Impact_01}}
]
properties {
 attenuationSettings=&/Fort_SleepyTime_Att
 bOverrideConcurrency=true
 maxDistance=107500
 soundClassObject=&Fort_Ambient_Athena
 volumeMultiplier=1.663
 attenuationOverrides {distanceAlgorithm=NaturalSound, falloffDistance=3800, lPFRadiusMax=3000, lPFRadiusMin=2000, attenuationShapeExtents {x=2200}}
 concurrencyOverrides {maxCount=4, resolutionRule=StopOldest, volumeScale=0.5}
}
