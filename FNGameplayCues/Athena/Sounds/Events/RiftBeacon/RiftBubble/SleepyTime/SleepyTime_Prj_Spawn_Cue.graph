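# Sound cue for the SleepyTime projectile spawn sound, part of the Athena RiftBeacon RiftBubble event audio.
# A Mixer passes through a single wave player for the SleepyTime spawn wave, with its second input left empty.
# The cue boosts volume, plays in the Athena ambient sound class and uses the shared SleepyTime attenuation asset.
# Its own concurrency override caps simultaneous instances at a small count.
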SoundCue /FNGameplayCues/Athena/Sounds/Events/RiftBeacon/RiftBubble/SleepyTime/SleepyTime_Prj_Spawn_Cue
root=Mixer_0
nodes [
 {id=Mixer_0, children [WavePlayer_2,null], properties {inputVolume [1,1]}}
 {id=WavePlayer_2, properties {soundWaveAssetPtr=&/SleepyTime_Spawn_01}}
]
properties {
 attenuationSettings=&/Fort_SleepyTime_Att
 bOverrideConcurrency=true
 maxDistance=107500
 soundClassObject=&Fort_Ambient_Athena
 volumeMultiplier=4.158
 attenuationOverrides {distanceAlgorithm=NaturalSound, falloffDistance=3800, lPFRadiusMax=3000, lPFRadiusMin=2000, attenuationShapeExtents {x=2200}}
 concurrencyOverrides {maxCount=3}
}
